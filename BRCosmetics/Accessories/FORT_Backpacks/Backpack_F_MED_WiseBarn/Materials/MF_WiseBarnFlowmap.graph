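# Material function for the WiseBarn backpack's orb center that produces an animated flowmap distortion as a direction vector and a UV offset.
# A turbulence texture pans on scaled input UVs in opposite directions for front and back faces, with explicit derivatives from tiled texcoords for mip selection.
# The remapped RG, damped at grazing angles, becomes a fresnel-weighted view-space vector transformed to world space, plus a separately scaled UV offset.
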MaterialFunction /BRCosmetics/Accessories/FORT_Backpacks/Backpack_F_MED_WiseBarn/Materials/MF_WiseBarnFlowmap
graph {nodes [
  n1 = NamedRerouteUsage()
  n2 = AppendVector(A=Constant() {r=0.5}, B=ScalarParameter() {defaultValue=4, group=WiseBarnOrbCenter, parameterName=Flowmap UV Scale, sliderMax=1})
  n3 = Multiply(A=ScalarParameter() {defaultValue=0.1, group=WiseBarnOrbCenter, parameterName=Flowmap Panner Speed}, B=TwoSidedSign())
  n4 = Constant()
  n5 = TextureCoordinate() {uTiling=3, vTiling=3}
  n6 = TextureSampleParameter2D!(
   UVs=Panner(Coordinate=Multiply(A=@n1, B=@n2), Speed=AppendVector(A=@n3, B=@n4)) {bFractionalPart=true}
   "DDX(UVs)"=DDX(Value=@n5)
   "DDY(UVs)"=DDY(Value=@n5)
   Apply View MipBias
  ) {
   group=WiseBarnOrbCenter
   mipValueMode=TMVM_Derivative
   parameterName=Flowmap
   samplerType=SAMPLERTYPE_LinearColor
   texture=/Game/Effects/Fort_Effects/Textures/Plasma/T_Turbulence_AdditiveWarp_Linear
  }
  n7 = Abs(None=DotProduct(A=PixelNormalWS(), B=CameraVectorWS()))
  n8 = Multiply(A=ConstantBiasScale(None=ComponentMask(None=@n6.RGB) {g=true, r=true}) {bias=-0.5, scale=2}, B=SquareRoot(None=@n7))
  n9 = ScalarParameter() {defaultValue=0.5, group=WiseBarnOrbCenter, parameterName=Fresnel Flowmap Intensity Max}
  n10 = ScalarParameter() {defaultValue=0.1, group=WiseBarnOrbCenter, parameterName=Fresnel Flowmap Intensity Min}
  n11 = AppendVector(A=Multiply(A=@n8, B=LinearInterpolate(A=@n9, B=@n10, Alpha=@n7)), B=Constant())
  FunctionOutput(None=NamedRerouteDeclaration(None=Transform(None=@n11) {transformSourceType=TRANSFORMSOURCE_View}) {name=Flowmap, nodeColor=<rgba 0.1016 1 0 1>}) {outputName=Flowmap}
  n12 = ScalarParameter() {defaultValue=0.066, parameterName=UV Flowmap Intensity}
  FunctionOutput(None=NamedRerouteDeclaration(None=Multiply(A=@n12, B=@n8)) {name=FlowmapUVs, nodeColor=<rgba 1 0.5625 0 1>}) {outputName=FlowmapUVs}
  NamedRerouteDeclaration(None=FunctionInput(Preview=TextureCoordinate()) {inputName=In_UVs, inputType=FunctionInput_Vector2}) {name=In_UVs, nodeColor=<rgba 0 1 0.9531 1>}
 ]}
