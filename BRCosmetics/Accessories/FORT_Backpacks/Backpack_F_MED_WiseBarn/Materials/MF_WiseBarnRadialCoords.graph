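# Material function for the WiseBarn backpack that builds view-dependent radial UV coordinates from the pixel normal and camera vector.
# U is a powered normal-to-camera dot product, acting as distance distortion.
# V approximates an angle without arctangent, using cross and dot products of the flattened camera-space direction against a fixed axis to select and remap each half-turn.
# An SCurve smooths angular distortion, bypassed at Low and Medium quality.
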MaterialFunction /BRCosmetics/Accessories/FORT_Backpacks/Backpack_F_MED_WiseBarn/Materials/MF_WiseBarnRadialCoords
graph {nodes [
  n1 = Power(Base=DotProduct(A=PixelNormalWS(), B=CameraVectorWS())) {desc=Radial coordinate distortion}
  n2 = Subtract(A=PixelNormalWS(), B=CameraVectorWS())
  n3 = Normalize(VectorInput=Multiply(
    A=Transform(None=@n2) {transformSourceType=TRANSFORMSOURCE_World, transformType=TRANSFORM_Camera}
    B=Constant3Vector() {constant=<rgba 1 1 0 0>}
   ))
  n4 = Constant3Vector() {constant=<rgba 0 1 0 0>}
  n5 = ComponentMask(None=CrossProduct(A=@n3, B=@n4)) {b=true}
  n6 = Abs(None=@n5)
  n7 = MaterialFunctionCall(In=@n6, Power=Constant() {r=0.75}) {
   desc="Kind of optional, but smoothes out\r\nsome nonlinear distortion"
   materialFunction=&SCurve
  }
  n8 = Multiply(A=QualitySwitch(Default=@n7.Result, Low=@n6, Medium=@n6), B=Sign(None=@n5))
  n9 = DotProduct(A=@n3, B=@n4)
  n10 = LinearInterpolate(A=ConstantBiasScale(None=@n8) {bias=-3, scale=-0.25}, B=ConstantBiasScale(None=@n8) {scale=0.25}, Alpha=Saturate(None=Sign(None=@n9))) {constB=0}
  FunctionOutput(None=NamedRerouteDeclaration(None=AppendVector(A=@n1, B=@n10) {desc=Output UVs}) {name=RadialCoords, nodeColor=<rgba 0.7578 0 1 1>}) {outputName=RadialCoords}
 ]}
