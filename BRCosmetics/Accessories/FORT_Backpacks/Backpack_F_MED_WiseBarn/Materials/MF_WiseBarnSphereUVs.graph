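# Material function for the WiseBarn backpack that outputs fake-sphere UVs on a camera-facing virtual plane.
# The plane is centered on the component pivot, with the camera vector as its normal and the X axis from that vector crossed with view-space up.
# The plane UVs, sized by a UV Scale parameter, feed MF_Sphere, borrowed from the PitGlass backpack, whose UV result becomes the SphereUVs output.
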MaterialFunction /BRCosmetics/Accessories/FORT_Backpacks/Backpack_F_MED_WiseBarn/Materials/MF_WiseBarnSphereUVs
graph {nodes [
  n1 = CameraVectorWS()
  n2 = Transform(None=Constant3Vector() {constant=<rgba 0 1 0 0>}) {transformSourceType=TRANSFORMSOURCE_View}
  n3 = MaterialFunctionCall(
   Plane Normal Axis=@n1
   Plane X Axis=Normalize(VectorInput=CrossProduct(A=@n1, B=@n2))
   Plane Center=MaterialFunctionCall() {materialFunction=&ComponentPivotLocation}.Result
   UV Scale=ScalarParameter() {defaultValue=50, parameterName=UV Scale}
  ) {materialFunction=&VirtualPlaneCoordinates}
  FunctionOutput(None=NamedRerouteDeclaration(None=MaterialFunctionCall!(UV=@n3.UVs, Size, Curvature) {materialFunction=/Game/Accessories/FORT_Backpacks/Backpack_M_MED_PitGlass/FX/Materials/MF_Sphere}.UV) {name=SphereUVs, nodeColor=<rgba 1 0 0.7969 1>}) {outputName=SphereUVs}
 ]}
