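# Material function that builds the emissive look of the WiseBarn backpack's crystal-ball orb, with a glowing center, electric noise and animated thin-film glitter.
# Center glow detail blends three animated texture channels, supported by camera-facing virtual-plane UVs anchored to a Blueprint-set socket position.
# View-dependent radial coordinates, Fresnel masks and flowmap-bent normals shape the layers.
# Medium and low quality skip the flowmap, second noise sample and thin-film boost.
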MaterialFunction /BRCosmetics/Accessories/FORT_Backpacks/Backpack_F_MED_WiseBarn/Materials/MF_WiseBarn_Orb
graph {nodes [
  n1 = VectorParameter() {group=WiseBarnOrbCenter, parameterName=Color C, defaultValue=<rgba 0 0 0 2>}
  n2 = Normalize(VectorInput=Add(A=NamedRerouteUsage(), B=PixelNormalWS()))
  n3 = VertexNormalWS()
  n4 = Abs(None=DotProduct(A=QualitySwitch(Default=@n2, Low=@n3, Medium=@n3) {desc=Skipping flowmap on med/low}, B=CameraVectorWS()))
  n5 = OneMinus(None=@n4)
  n6 = NamedRerouteUsage()
  Add(A=LinearInterpolate(A=@n1.RGB, B=NamedRerouteUsage(), Alpha=Power(Base=@n5, Exp=@n1.A)), B=@n6) {constB=0}
  FunctionOutput(None=MaterialFunctionCall(Emissive=@n6) {materialFunction=&MF_FX_EmissiveAdjustments}.Result) {outputName=Emissive}
  n7 = ComponentMask(None=ScreenPosition().ViewportUV) {r=true}
  n8 = ScalarParameter() {defaultValue=1, parameterName=A/B Compare}
  LinearInterpolate(B=Constant3Vector() {constant=<rgba 1 0 0 0>}, Alpha=SphereMask(A=@n7, B=@n8) {attenuationRadius=0.01})
  MaterialFunctionCall!(Blend=@n6, Base=NamedRerouteUsage()) {materialFunction=/Engine/Functions/Engine_MaterialFunctions03/Blends/Blend_Screen}
  n9 = Multiply(A=@n6, B=ScalarParameter() {defaultValue=1, group=WiseBarnOrbCenter, parameterName=CenterGlowMultiplier})
  n10 = VectorParameter() {group=WiseBarnOrbCenter, parameterName=Color Lightning Fresnel, defaultValue=<rgba 1.3 0 3 0.1>}
  n11 = ScalarParameter() {defaultValue=3, group=WiseBarnOrbCenter, parameterName=Lightning Edge Fade Exponent}
  n12 = LinearInterpolate(
   A=VectorParameter() {group=WiseBarnOrbCenter, parameterName=Color Lightning Falloff, defaultValue=<rgba 0.015 0 0.9 3>}.RGB
   B=VectorParameter() {group=WiseBarnOrbCenter, parameterName=Color Lightning, defaultValue=<rgba 0.07 0.845 1 0>}.RGB
   Alpha=Multiply(A=@n9, B=Power(Base=@n4, Exp=@n11) {constExponent=3})
  )
  n13 = MaterialFunctionCall(f4=Normalize(VectorInput=VectorParameter() {group=WiseBarnOrbCenter, parameterName=ColorLightningSpectrumSensitivity, defaultValue=<rgba 1 2 0 4>}.RGBA)) {materialFunction=&AddComponents}
  Multiply(A=@n9, B=LinearInterpolate(A=@n10.RGB, B=Multiply(A=@n12, B=@n13.f4), Alpha=Power(Base=@n4, Exp=@n10.A)))
  n14 = ComponentMask(None=NamedRerouteUsage()) {g=true}
  n15 = Multiply(A=Time(), B=ScalarParameter() {defaultValue=0.1, group=WiseBarnOrbCenter, parameterName=CustomThinFilmCircularPanSpeed})
  n16 = CurveAtlasRowParameter(CurveTime=Frac(None=Add(A=@n14, B=@n15))) {
   atlas=/Game/Weapons/FORT_Melee/Pickaxe_RelayStick/Textures/CA_RelayStick
   curve=/Game/Weapons/FORT_Melee/Pickaxe_RelayStick/Textures/C_RelayStick_CircularFresnel
   defaultValue=1
   group=WiseBarnOrbCenter
   parameterName=CustomThinFilmRadialGradient
  }
  Multiply(A=@n16, B=VectorParameter() {group=WiseBarnOrbCenter, parameterName=OrbBackgroundTint, defaultValue=<rgba 0.0625 0.6958 2 1>}.RGB)
  n17 = Saturate(None=Divide(A=@n5, B=ScalarParameter() {defaultValue=0.75, group=WiseBarnOrbCenter, parameterName=Center Glow Size}))
  NamedRerouteDeclaration(None=Power(Base=OneMinus(None=@n17), Exp=ScalarParameter() {defaultValue=22, group=WiseBarnOrbCenter, parameterName=Center Glow Exponent})) {name=Center Glow Mask, nodeColor=<rgba 1 0.8203 0 1>}
  n18 = VectorParameter() {desc="value is from socket, and set by blueprint", parameterName=WiseBarnBackpackPosition}
  n19 = Transform(None=Add(
    A=Transform(None=@n18.RGB) {transformSourceType=TRANSFORMSOURCE_World, transformType=TRANSFORM_Local}
    B=VectorParameter() {group=WiseBarnOrbCenter, parameterName=ObjectPositionOffset, defaultValue=<rgba 0 2 3.8 0>}.RGB
   )) {transformSourceType=TRANSFORMSOURCE_Local}
  n20 = Subtract(A=@n19, B=CameraPositionWS())
  NamedRerouteDeclaration(None=MaterialFunctionCall(
    Plane Normal Axis=Normalize(VectorInput=@n20)
    Plane Center=@n19
    UV Scale=ScalarParameter() {defaultValue=200, group=WiseBarnOrbCenter, parameterName=UVScaleCenter}
   ) {materialFunction=&VirtualPlaneCoordinates}.UVs) {name=VirtualPlaneUVs, nodeColor=<rgba 0.5 0 1 1>}
  n21 = TextureCoordinate()
  NamedRerouteDeclaration(None=FunctionInput(Preview=MaterialFunctionCall!(In_UVs=@n21) {materialFunction=&/MF_WiseBarnFlowmap}.Flowmap) {bUsePreviewValueAsDefault=true, inputName=In_Flowmap, sortPriority=4}) {name=In_Flowmap, nodeColor=<rgba 1 0 0.0234 1>}
  n22 = TextureObjectParameter() {
   parameterName=CenterGlow
   samplerType=SAMPLERTYPE_LinearColor
   texture=/BRCosmetics/Accessories/FORT_Backpacks/Backpack_F_MED_WiseBarn/Textures/T_WindBarn_Center
  }
  NamedRerouteDeclaration(None=@n22) {name=CenterGlowTexture, nodeColor=<rgba 0 0.1094 1 1>}
  NamedRerouteDeclaration(None=MaterialFunctionCall!(UV Coord=NamedRerouteUsage(), LenghtScale, RadialScale) {materialFunction=/Game/Global/MaterialFunctions/TechArt/MF_PolarCoordinates}.Result) {name=PolarCoords, nodeColor=<rgba 0.5937 1 0 1>}
  MaterialExpressionNamedRerouteUsage_27 = NamedRerouteUsage()
  n23 = Convert(None=@MaterialExpressionNamedRerouteUsage_27) {
   nodeName=Break Float2
   convertInputs [{
     type=Vector2
     expressionInput {
      expression=&self:MaterialExpressionNamedRerouteUsage_27
      inputName=None
      mask=0
      maskA=0
      maskB=0
      maskG=0
      maskR=0
      outputIndex=0
     }
    }]
   convertOutputs [{type=Scalar}, {type=Scalar}]
  }
  n24 = Divide(A=@n23)
  n25 = Sine(None=@n24)
  n26 = Cosine(None=@n24)
  NamedRerouteDeclaration(None=Multiply(A=@n23, B=AppendVector(A=@n25, B=@n26))) {name=PolarToCartesian, nodeColor=<rgba 1 0 0.1172 1>}
  n27 = Arcsine!(None=DotProduct(A=PixelNormalWS(), B=CameraVectorWS()))
  n28 = MaterialFunctionCall() {materialFunction=&Pi}
  n29 = Subtract(A=PixelNormalWS(), B=CameraVectorWS())
  n30 = Normalize(VectorInput=Multiply(
    A=Transform(None=@n29) {transformSourceType=TRANSFORMSOURCE_World, transformType=TRANSFORM_Camera}
    B=Constant3Vector() {constant=<rgba 1 1 0 0>}
   ))
  n31 = Constant3Vector() {constant=<rgba 0 1 0 0>}
  n32 = ComponentMask(None=CrossProduct(A=@n30, B=@n31)) {b=true}
  n33 = Abs(None=@n32)
  n34 = Constant() {r=0.2}
  n35 = Multiply(
   A=Multiply(A=Arcsine(None=@n33), B=MaterialFunctionCall(Multiplier=@n34) {materialFunction=&Pi}.Result)
   B=Sign(None=@n32)
  )
  n36 = ConstantBiasScale(None=@n35) {bias=-3, scale=-0.25}
  n37 = ConstantBiasScale(None=@n35) {scale=0.25}
  n38 = Saturate(None=Sign(None=DotProduct(A=@n30, B=@n31)))
  n39 = NamedRerouteDeclaration(None=AppendVector(A=Divide(A=@n27, B=@n28.Result), B=LinearInterpolate(A=@n36, B=@n37, Alpha=@n38) {constB=0}) {desc=Output UVs}) {name=NewRadialCoords, nodeColor=<rgba 1 0.9375 0 1>}
  n40 = Constant2Vector() {r=0.5}
  n41 = MaterialFunctionCall(Multiplier=Constant() {r=0.5}) {materialFunction=&Pi}
  n42 = Constant() {r=2}
  NamedRerouteDeclaration(None=Multiply(A=Subtract(A=@n39, B=@n40), B=AppendVector(A=@n41.Result, B=@n42))) {name=RemappedPolarCoords, nodeColor=<rgba 0.6172 0 1 1>}
  NamedRerouteDeclaration(None=LinearInterpolate(A=NamedRerouteUsage(), B=NamedRerouteUsage(), Alpha=Constant())) {name=NewPolarCoords, nodeColor=<rgba 0 1 0.2969 1>}
  n43 = ScalarParameter() {defaultValue=0.1, parameterName=RedChannelSpeedX}
  n44 = ScalarParameter() {parameterName=RedChannelSpeedY}
  n45 = TextureSample(UVs=Panner(Coordinate=NamedRerouteUsage(), Speed=AppendVector(A=@n43, B=@n44)) {speedX=-0.1}, Tex=NamedRerouteUsage()) {samplerType=SAMPLERTYPE_LinearColor, texture=&T_WindBarn_Center}
  n46 = ScalarParameter() {defaultValue=0.0813, parameterName=GreenChannelSpeedX}
  n47 = ScalarParameter() {parameterName=GreenChannelSpeedY}
  n48 = TextureSample(UVs=Panner(Coordinate=NamedRerouteUsage(), Speed=AppendVector(A=@n46, B=@n47)) {speedX=-0.1}, Tex=NamedRerouteUsage()) {samplerType=SAMPLERTYPE_LinearColor, texture=&T_WindBarn_Center}
  n49 = Multiply(
   A=MaterialFunctionCall(
    A=VectorParameter() {parameterName=RedChannelA, defaultValue=<rgba 0.0036 0.018 0.0522 0>}.RGB
    B=VectorParameter() {parameterName=RedChannelB, defaultValue=<rgba 0.1099 0.0543 1 1>}.RGB
    C=VectorParameter() {parameterName=RedChannelC, defaultValue=<rgba 0.8934 0.5565 2 1>}.RGB
    Alpha=NamedRerouteDeclaration(None=@n45.R) {name=RedChannel, nodeColor=<rgba 1 0 0 1>}
   ) {materialFunction=&Lerp_3Color}.Result
   B=MaterialFunctionCall(
    A=VectorParameter() {parameterName=GreenChannelA, defaultValue=<rgba 0.0222 0.0036 0.0522 0>}.RGB
    B=VectorParameter() {parameterName=GreenChannelB, defaultValue=<rgba 0.4783 0.0042 0.4247 1>}.RGB
    C=VectorParameter() {parameterName=GreenChannelC, defaultValue=<rgba 0.0941 0.7474 1.0826 1>}.RGB
    Alpha=NamedRerouteDeclaration(None=@n48.G) {name=GreenChannel, nodeColor=<rgba 0 1 0 1>}
   ) {materialFunction=&Lerp_3Color}.Result
  )
  n50 = ComponentMask(None=Multiply(A=NamedRerouteUsage()) {constB=0.2}) {g=true, r=true}
  n51 = Multiply(A=AppendVector(A=NamedRerouteUsage(), B=NamedRerouteUsage())) {constB=0.067}
  n52 = Subtract(A=Add(A=Add(A=@n50, B=@n51), B=NamedRerouteUsage())) {constB=-0.5}
  n53 = Time()
  n54 = TextureSample(
   UVs=MaterialFunctionCall(UVs=Rotator(Coordinate=@n52, Time=@n53), Texture Scale=Constant() {r=1.3}) {materialFunction=&ScaleUVsByCenter}.UVs
   Tex=NamedRerouteUsage()
  ) {samplerType=SAMPLERTYPE_LinearColor, texture=&T_WindBarn_Center}
  n55 = VectorParameter() {parameterName=BlueChannelA, defaultValue=<rgba 0.0222 0.0036 0.0522 0>}
  n56 = VectorParameter() {parameterName=BlueChannelB, defaultValue=<rgba 0.3304 0.0029 0.2934 1>}
  n57 = VectorParameter() {parameterName=BlueChannelC, defaultValue=<rgba 0 0.0948 0.4957 1>}
  n58 = SmoothStep(Value=@n54.B) {constMax=2}
  n59 = Multiply(A=ScalarParameter() {defaultValue=2, parameterName=BlueChannelLerpMultiplier}, B=Saturate(None=NamedRerouteUsage()))
  NamedRerouteDeclaration(None=LinearInterpolate(
    A=Multiply(A=@n49, B=@n54.B)
    B=MaterialFunctionCall(A=@n55.RGB, B=@n56.RGB, C=@n57.RGB, Alpha=@n58) {materialFunction=&Lerp_3Color}.Result
    Alpha=Multiply(A=@n58, B=@n59) {constB=0.67}
   )) {name=CenterGlowDetail, nodeColor=<rgba 0 0.2266 1 1>}
  n60 = Constant() {desc=input length scale, r=1}
  n61 = Constant() {r=0.5}
  n62 = MaterialFunctionCall() {materialFunction=&ObjectPivotPoint}
  n63 = Add(
   A=AppendVector(A=@n61, B=@n61)
   B=NamedRerouteDeclaration(None=MaterialFunctionCall!(Position=@n62.Object Pivot Location, UVRadialScale) {materialFunction=/Game/Weapons/WeaponSkins/MaterialFunction/ObjectStabilizedScreenSpaceUVs}.StabilizedUVs) {name=InputCartesianUVs, nodeColor=<rgba 1 0 0.0234 1>}
  )
  n64 = ComponentMask(None=@n63) {r=true}
  n65 = Multiply(A=@n60, B=@n64)
  n66 = ComponentMask(None=@n63) {g=true}
  n67 = Multiply(A=@n60, B=@n66)
  n68 = Add(A=Multiply(A=@n65, B=@n65), B=Multiply(A=@n67, B=@n67))
  n69 = Arctangent2Fast(Y=@n66, X=@n64)
  n70 = MaterialFunctionCall(Multiplier=Constant() {desc=input radial scale, r=1}) {materialFunction=&Pi}
  NamedRerouteDeclaration(None=AppendVector(A=SquareRoot(None=@n68), B=Divide(A=@n69, B=@n70.Result))) {name=ObjectPolarCoords, nodeColor=<rgba 0 0.3437 1 1>}
  n71 = NamedRerouteUsage()
  n72 = VectorParameter() {group=WiseBarnOrbCenter, parameterName="OrbRadialCoordScales (UVUV)", defaultValue=<rgba 0.1366 2 0.174 1>}
  n73 = AppendVector(A=@n72.R, B=@n72.R)
  n74 = VectorParameter() {group=WiseBarnOrbCenter, parameterName="OrbRadialCoordPanSpeeds (UVUV)", defaultValue=<rgba 0.0333 0.0333 -0.0267 -0.0233>}
  n75 = TextureCoordinate() {uTiling=3, vTiling=3}
  n76 = DDX(Value=@n75)
  n77 = DDY(Value=@n75)
  n78 = TextureSampleParameter2D!(
   UVs=Panner(Coordinate=Multiply(A=@n71, B=@n73) {constB=3}, Speed=AppendVector(A=@n74.R, B=@n74.R)) {speedX=-1.1, speedY=-0.067}
   "DDX(UVs)"=@n76
   "DDY(UVs)"=@n77
   Apply View MipBias
  ) {
   group=WiseBarnOrbCenter
   mipValueMode=TMVM_Derivative
   parameterName=NoiseA
   samplerType=SAMPLERTYPE_LinearColor
   texture=/Game/Characters/Player/Male/Medium/Bodies/M_MED_SRDrift/Textures/T_ElectricNoise
  }
  n79 = NamedRerouteUsage()
  n80 = AppendVector(A=@n72.B, B=@n72.B)
  n81 = TextureSampleParameter2D(
   UVs=Panner(Coordinate=Multiply(A=@n79, B=@n80), Time=Time(), Speed=AppendVector(A=@n74.B, B=@n74.B)) {speedX=-0.33, speedY=-0.053}
   "DDX(UVs)"=@n76
   "DDY(UVs)"=@n77
  ) {
   group=WiseBarnOrbCenter
   mipValueMode=TMVM_Derivative
   parameterName=NoiseB
   samplerType=SAMPLERTYPE_LinearColor
   texture=&T_ElectricNoise
  }
  n82 = Add(A=@n78.R, B=@n81.R)
  n83 = OneMinus(None=Power(Base=NamedRerouteUsage(), Exp=ScalarParameter() {defaultValue=1, group=WiseBarnOrbCenter, parameterName=CenterFresnelExponent}) {constExponent=0.1})
  n84 = ScalarParameter() {defaultValue=1, group=WiseBarnOrbCenter, parameterName=CenterStrength}
  NamedRerouteDeclaration(None=Multiply(A=QualitySwitch(Default=@n82, Low=@n78.R, Medium=@n78.R), B=Multiply(A=@n83, B=@n84))) {name=NoiseColorBackground, nodeColor=<rgba 0 1 0.5547 1>}
  n85 = Add(A=NamedRerouteUsage(), B=@n16)
  n86 = ScalarParameter() {defaultValue=2, group=WiseBarnOrbCenter, parameterName=ThinFilmBoost}
  NamedRerouteDeclaration(None=QualitySwitch(Default=Multiply(A=@n85, B=@n86) {constB=12}, Low=@n16, Medium=@n16)) {name=SpectralThinFilm, nodeColor=<rgba 0.8516 1 0 1>}
  NamedRerouteDeclaration(None=FunctionInput(Preview=Fresnel() {baseReflectFraction=0, exponent=1}) {bUsePreviewValueAsDefault=true, inputName=Fresnel, inputType=FunctionInput_Scalar}) {name=Fresnel, nodeColor=<rgba 0.4531 0 1 1>}
  n87 = Power(Base=DotProduct(A=PixelNormalWS(), B=CameraVectorWS())) {constExponent=1, desc=Radial coordinate distortion}
  n88 = Subtract(A=PixelNormalWS(), B=CameraVectorWS())
  n89 = Normalize(VectorInput=Multiply(
    A=Transform(None=@n88) {transformSourceType=TRANSFORMSOURCE_World, transformType=TRANSFORM_Camera}
    B=Constant3Vector() {constant=<rgba 1 1 0 0>}
   ))
  n90 = Constant3Vector() {constant=<rgba 0 1 0 0>}
  n91 = ComponentMask(None=CrossProduct(A=@n89, B=@n90)) {b=true}
  n92 = Abs(None=@n91)
  n93 = MaterialFunctionCall(In=@n92, Power=Constant() {r=0.75}) {
   desc="Kind of optional, but smoothes out\r\nsome nonlinear distortion"
   materialFunction=&SCurve
  }
  n94 = Multiply(A=QualitySwitch(Default=@n93.Result, Low=@n92, Medium=@n92), B=Sign(None=@n91))
  n95 = DotProduct(A=@n89, B=@n90)
  n96 = LinearInterpolate(A=ConstantBiasScale(None=@n94) {bias=-3, scale=-0.25}, B=ConstantBiasScale(None=@n94) {scale=0.25}, Alpha=Saturate(None=Sign(None=@n95))) {constB=0}
  NamedRerouteDeclaration(None=FunctionInput(Preview=AppendVector(A=@n87, B=@n96) {desc=Output UVs}) {bUsePreviewValueAsDefault=true, inputName=RadialUVCoordinates, inputType=FunctionInput_Vector2, sortPriority=3}) {name=RadialCoords, nodeColor=<rgba 0.7578 0 1 1>}
  NamedRerouteDeclaration(None=NamedRerouteUsage()) {name=Flowmap, nodeColor=<rgba 0.1016 1 0 1>}
  n97 = NamedRerouteUsage()
  n98 = ScalarParameter() {defaultValue=0.5, group=WiseBarnOrbCenter, parameterName=CustomThinFilmWarp, sortPriority=4}
  n99 = Add(
   A=Multiply(A=ScalarParameter() {defaultValue=-0.1, group=WiseBarnOrbCenter, parameterName=CustomThinFilmPanSpeed}, B=Time())
   B=Multiply(
    A=ScalarParameter() {defaultValue=0.5, group=WiseBarnOrbCenter, parameterName=CustomThinFilmScale, sortPriority=5}
    B=Power(Base=@n97, Exp=@n98)
   )
  )
  n100 = Subtract(
   A=@n14
   B=Multiply(A=Time(), B=ScalarParameter() {defaultValue=0.1, group=WiseBarnOrbCenter, parameterName=CustomThinFilmCircularMaskRotateSpeed})
  )
  n101 = TextureCoordinate()
  n102 = TextureSampleParameter2D(UVs=Multiply(A=AppendVector(A=@n99, B=@n100)) {constB=2}, "DDX(UVs)"=DDX(Value=@n101), "DDY(UVs)"=DDY(Value=@n101)) {
   group=WiseBarnOrbCenter
   mipValueMode=TMVM_Derivative
   parameterName=ThinFilmPattern
   texture=/BRCosmetics/Accessories/FORT_Backpacks/Backpack_F_MED_WiseBarn/Textures/T_WindBarn_Glitter
  }
  NamedRerouteDeclaration(None=@n102.RGB) {name=ThinFilmPatternTexture, nodeColor=<rgba 1 0.7266 0 1>}
  MaterialExpressionNamedRerouteUsage_6 = NamedRerouteUsage()
  n103 = Convert(None=@MaterialExpressionNamedRerouteUsage_6) {
   nodeName=Break Float3
   convertInputs [{
     type=Vector3
     expressionInput {
      expression=&self:MaterialExpressionNamedRerouteUsage_6
      inputName=None
      mask=0
      maskA=0
      maskB=0
      maskG=0
      maskR=0
      outputIndex=0
     }
    }]
   convertOutputs [{type=Scalar}, {type=Scalar}, {type=Scalar}]
  }
  n104 = Multiply(A=VectorParameter() {group=WiseBarnOrbCenter, parameterName=RedChannelColor, defaultValue=<rgba 0.0625 0.6958 2 1>}.RGB, B=@n103)
  n105 = Multiply(A=VectorParameter() {group=WiseBarnOrbCenter, parameterName=GreenChannelColor, defaultValue=<rgba 0.8038 0.0625 2 1>}.RGB, B=@n103)
  n106 = VectorParameter() {group=WiseBarnOrbCenter, parameterName=BlueChannelColor, defaultValue=<rgba 0.0625 1.444 2 1>}
  NamedRerouteDeclaration(None=Add(A=Add(A=@n104, B=@n105), B=Multiply(A=@n106.RGB, B=@n103))) {name=BitsFloatingInCrystalBallColor, nodeColor=<rgba 0.4062 0 1 1>}
  NamedRerouteUsage()
  NamedRerouteUsage()
  NamedRerouteUsage()
  n107 = NamedRerouteUsage()
  n108 = PixelNormalWS()
  n109 = VertexNormalWS()
  n110 = QualitySwitch(Default=Normalize(VectorInput=Add(A=@n107, B=@n108)), Low=@n109, Medium=@n109) {desc=Skipping flowmap on med/low}
  n111 = CameraVectorWS()
  OneMinus(None=Abs(None=DotProduct(A=@n110, B=@n111)))
  Step(Y=@n7, X=@n8)
 ]}
