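# Material function for the WiseBarn backpack's crystal ball, overriding base color, emissive and world position offset on incoming material attributes, with blue-channel mask blending.
# Flowmapped sphere UVs, three-layer panned lightning streaks and an orb subfunction build the glow, combined through screen blending.
# Fresnel-driven thin film and cubemap reflections add a glassy shell, while vertices offset along normal and camera direction.
# Quality switches skip the flowmap on low and medium.
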MaterialFunction /BRCosmetics/Accessories/FORT_Backpacks/Backpack_F_MED_WiseBarn/Materials/MF_WiseBarnCrystalBall
graph {nodes [
  Add(A=NamedRerouteUsage())
  n1 = GetMaterialAttributes(None=NamedRerouteUsage()) {attributeGetTypes [#1, #2]}
  n2 = NamedRerouteUsage()
  n3 = LinearInterpolate(A=@n2, B=NamedRerouteUsage(), Alpha=ComponentMask(None=NamedRerouteUsage()) {b=true})
  FunctionOutput(None=SetMaterialAttributes!(
    MaterialAttributes=@n1.MaterialAttributes
    EmissiveColor=NamedRerouteUsage()
    WorldPositionOffset=NamedRerouteUsage()
    BaseColor=NamedRerouteDeclaration(None=@n3) {name=Out_BaseColor, nodeColor=<rgba 1 0.8906 0 1>}
   ) {attributeSetTypes [#1, #3, #2]})
  MaterialFunctionCall() {materialFunction=&/MF_WiseBarnRadialCoords}
  n4 = ScalarParameter() {defaultValue=3, parameterName=Blended Texture Exponent}
  n5 = VectorParameter() {parameterName=Color Lightning Fresnel, defaultValue=<rgba 1.3 0 3 0.1>}
  n6 = VectorParameter() {parameterName=Color Lightning Falloff, defaultValue=<rgba 0.015 0 0.9 3>}
  n7 = VectorParameter() {parameterName=Color Lightning, defaultValue=<rgba 0.07 0.845 1 0>}
  n8 = MaterialFunctionCall(Power=@n4) {materialFunction=&SCurve}
  n9 = NamedRerouteUsage()
  n10 = NamedRerouteUsage()
  Multiply(
   A=Power(Exp=@n4)
   B=LinearInterpolate(A=@n5.RGB, B=LinearInterpolate(A=@n6.RGB, B=@n7.RGB, Alpha=@n8.Result), Alpha=Power(Base=@n9, Exp=@n10))
  )
  n11 = NamedRerouteUsage()
  Multiply(A=@n11, B=VectorParameter() {group=06_Thin Film, parameterName=ThinFilmColorMultiplier, defaultValue=<rgba 1 1 1 0>}.RGB)
  Multiply(A=@n2, B=ScalarParameter() {defaultValue=1, parameterName=BaseColorMultiplier})
  n12 = NamedRerouteUsage()
  n13 = ScalarParameter() {defaultValue=0.75, group=WiseBarnCrystalBall, parameterName=Center Glow Size}
  n14 = Saturate(None=Divide(A=OneMinus(None=@n12), B=Saturate(None=@n13)))
  NamedRerouteDeclaration(None=Power(Base=OneMinus(None=@n14), Exp=ScalarParameter() {defaultValue=22, group=WiseBarnCrystalBall, parameterName=Center Glow Exponent})) {name=Center Glow Mask, nodeColor=<rgba 1 0.8203 0 1>}
  n15 = Normalize(VectorInput=Subtract(A=CameraPositionWS(), B=WorldPosition().XYZ))
  n16 = ScalarParameter() {parameterName=Camera Offset}
  n17 = VertexNormalWS()
  n18 = ScalarParameter() {defaultValue=0.333, parameterName=Normal Offset}
  NamedRerouteDeclaration(None=Add(A=Multiply(A=@n15, B=@n16), B=Multiply(A=@n17, B=@n18))) {name=Out_WPO, nodeColor=<rgba 0 1 0.4844 1>}
  n19 = StaticBool() {value=true}
  n20 = MaterialFunctionCall!(
   UV=Divide(A=NamedRerouteUsage(), B=ScalarParameter() {defaultValue=2.618, parameterName=Caustics UV Scale})
   Rotation=Constant() {r=0.25}
   Speed=ScalarParameter() {defaultValue=0.066, parameterName=Caustics Speed}
   Time
   Rotate UVs=@n19
   Fractional Part=@n19
  ) {materialFunction=/Game/Accessories/FORT_Backpacks/Backpack_M_MED_PitGlass/FX/Materials/MF_3UVPanner}
  n21 = TextureObjectParameter() {
   parameterName=Streaks
   samplerType=SAMPLERTYPE_LinearGrayscale
   texture=/Game/Accessories/FORT_Backpacks/Backpack_M_MED_PitGlass/FX/Textures/T_Backpack_PitGlass_Noise
  }
  n22 = Add(
   A=Power(
    Base=TextureSample(UVs=@n20.UV1, Tex=@n21) {samplerSource=SSM_Wrap_WorldGroupSettings, texture=/Game/Effects/Fort_Effects/Textures/Tile/T_ShieldCaustics_Tile_Chromatic}.R
    Exp=NamedRerouteUsage()
   )
   B=Power(
    Base=TextureSample(UVs=@n20.UV2, Tex=@n21) {samplerSource=SSM_Wrap_WorldGroupSettings, texture=&T_ShieldCaustics_Tile_Chromatic}.R
    Exp=NamedRerouteUsage()
   )
  )
  n23 = Power(
   Base=TextureSample(UVs=@n20.UV3, Tex=@n21) {samplerSource=SSM_Wrap_WorldGroupSettings, texture=&T_ShieldCaustics_Tile_Chromatic}.R
   Exp=NamedRerouteUsage()
  )
  NamedRerouteDeclaration(None=Divide(A=Add(A=@n22, B=@n23)) {constB=3}) {name=LightningStreaks, nodeColor=<rgba 0 1 0.7188 1>}
  n24 = MaterialFunctionCall() {materialFunction=&MF_WiseBarnRadialCoords}
  n25 = NamedRerouteUsage()
  n26 = NamedRerouteUsage()
  n27 = VectorParameter() {group=WiseBarnCrystalBall, parameterName=LightningStreaksColor, defaultValue=<rgba 0.4769 0.0835 1.2 1>}
  NamedRerouteDeclaration(None=MaterialFunctionCall!(
    Blend=MaterialFunctionCall!(Fresnel, RadialUVCoordinates=@n24.RadialCoords, In_Flowmap=@n25) {materialFunction=&/MF_WiseBarn_Orb}.Emissive
    Base=Multiply(A=@n26, B=@n27.RGB)
   ) {materialFunction=/Engine/Functions/Engine_MaterialFunctions03/Blends/Blend_Screen}.Result) {name=CombinedColor, nodeColor=<rgba 1 0.3984 0 1>}
  n28 = NamedRerouteUsage()
  n29 = PixelNormalWS()
  n30 = VertexNormalWS()
  n31 = QualitySwitch(Default=Normalize(VectorInput=Add(A=@n28, B=@n29)), Low=@n30, Medium=@n30) {desc=Skipping flowmap on med/low}
  n32 = CameraVectorWS()
  NamedRerouteDeclaration(None=Abs(None=DotProduct(A=@n31, B=@n32))) {name=FlowmapDotCameraVector, nodeColor=<rgba 0 0.0859 1 1>}
  NamedRerouteDeclaration(None=FunctionInput() {inputName=In_MaterialAttribute, inputType=FunctionInput_MaterialAttributes}) {name=In_MaterialAttribute, nodeColor=<rgba 0 1 0.9531 1>}
  n33 = NamedRerouteUsage()
  NamedRerouteDeclaration(None=LinearInterpolate(A=@n1.EmissiveColor, B=NamedRerouteUsage(), Alpha=ComponentMask(None=@n33) {b=true})) {name=Out_Emissive, nodeColor=<rgba 0.3125 0 1 1>}
  NamedRerouteDeclaration(None=@n1.BaseColor) {name=In_BaseColor, nodeColor=<rgba 0 0.7891 1 1>}
  n34 = TextureSample() {
   samplerType=SAMPLERTYPE_Masks
   texture=/BRCosmetics/Accessories/FORT_Backpacks/Backpack_F_MED_WiseBarn/Textures/T_WiseBarn_Backpack_FX_TechArt
  }
  NamedRerouteDeclaration(None=FunctionInput(Preview=@n34.RGB) {inputName=In_TechArtMask2, sortPriority=2}) {name=In_TechArtMask2, nodeColor=<rgba 0.1279 0.4068 1 1>}
  n35 = MaterialFunctionCall!(In_UVs=NamedRerouteDeclaration(None=MaterialFunctionCall() {materialFunction=&/MF_WiseBarnSphereUVs}.SphereUVs) {name=SphereUVs, nodeColor=<rgba 1 0 0.7969 1>}) {materialFunction=&/MF_WiseBarnFlowmap}
  NamedRerouteDeclaration(None=@n35.Flowmap) {name=Flowmap_1, nodeColor=<rgba 0 1 0.1328 1>}
  NamedRerouteDeclaration(None=@n35.FlowmapUVs) {name=FlowmapUVs_1, nodeColor=<rgba 0.8516 0 1 1>}
  n36 = ScalarParameter() {defaultValue=0.75, group=06_Thin Film, parameterName=ThinFilmScale, sortPriority=5}
  n37 = Fresnel(
   ExponentIn=ScalarParameter() {defaultValue=1.5, parameterName=ExponentIn_2}
   BaseReflectFractionIn=ScalarParameter() {parameterName=BaseReflectFractionIn_2}
  )
  n38 = Power(Base=@n37, Exp=ScalarParameter() {defaultValue=0.5, group=06_Thin Film, parameterName=ThinFilmWarp, sortPriority=4})
  n39 = TextureSampleParameter2D(UVs=Add(A=Multiply(A=@n36, B=@n38), B=ScalarParameter() {defaultValue=0.3, group=06_Thin Film, parameterName=ThinFilmOffset})) {
   group=06_Thin Film
   parameterName=Thin Film Texture
   sortPriority=1
   texture=/Game/Global/Textures/Default/Thinfilm/T_ThinFilm_Spectrum_COLOR
  }
  n40 = ScalarParameter() {defaultValue=0.35, group=06_Thin Film, parameterName=ThinFilmIntensity, sortPriority=2}
  n41 = Power(Base=@n37, Exp=ScalarParameter() {defaultValue=0.75, group=06_Thin Film, parameterName=ThinFilmExponent, sortPriority=3})
  n42 = Multiply(A=LinearInterpolate() {constA=1, constB=0.01}, B=@n41)
  n43 = ScalarParameter() {group=06_Thin Film, parameterName=RoughnessInfluence}
  n44 = LinearInterpolate(B=Add(B=Multiply(A=@n39.RGB, B=@n40)), Alpha=Saturate(None=LinearInterpolate(A=@n41, B=@n42, Alpha=@n43)))
  n45 = ScalarParameter() {defaultValue=1.5, parameterName=ExponentIn}
  n46 = ScalarParameter() {defaultValue=0.04, parameterName=BaseReflectFractionIn}
  n47 = ReflectionVectorWS()
  n48 = Multiply(
   A=Multiply(
    A=Fresnel(ExponentIn=@n45, BaseReflectFractionIn=@n46) {baseReflectFraction=0, exponent=1}
    B=TextureSample(UVs=@n47) {texture=/Game/Packages/Fortress_Sky/SkyDome/MaterialInstances/SkyDome_S8/Cubemap_S8_Eve}.RGB
   )
   B=ScalarParameter() {defaultValue=1.5, parameterName=ReflectionStrength}
  )
  n49 = OneMinus(None=Fresnel(
    ExponentIn=ScalarParameter() {defaultValue=0.1, parameterName=ExponentIn_1}
    BaseReflectFractionIn=ScalarParameter() {parameterName=BaseReflectFractionIn_1}
   ) {baseReflectFraction=0, exponent=1})
  n50 = Max(A=ScalarParameter() {defaultValue=1, parameterName=FX Strength}) {constB=0}
  NamedRerouteDeclaration(None=LinearInterpolate(A=Add(A=@n44, B=@n48), B=@n11, Alpha=Multiply(A=@n49, B=@n50))) {name=FinalEmissive, nodeColor=<rgba 0 0.1562 1 1>}
  n51 = NamedRerouteUsage()
  n52 = NamedRerouteUsage()
  n53 = Transform(None=VertexNormalWS()) {transformSourceType=TRANSFORMSOURCE_World, transformType=TRANSFORM_Camera}
  n54 = VertexInterpolator(VS=ComponentMask(None=ConstantBiasScale(None=@n53)) {g=true, r=true})
  NamedRerouteDeclaration(None=QualitySwitch(Default=Add(A=@n51, B=@n52) {constB=0}, Low=@n54.PS, Medium=@n54.PS)) {name=FinalUVs, nodeColor=<rgba 0.4766 1 0 1>}
  NamedRerouteDeclaration(None=NamedRerouteUsage()) {name=FlowMap, nodeColor=<rgba 0 1 0.8828 1>}
  NamedRerouteDeclaration(None=NamedRerouteUsage()) {name=FlowmapUVs, nodeColor=<rgba 1 0.5625 0 1>}
  NamedRerouteDeclaration(None=Constant() {r=2}) {name=LightningAnimationExponent, nodeColor=<rgba 1 0.7266 0 1>}
  NamedRerouteDeclaration(None=@n5.A) {name=PowerExponentFromFresnel, nodeColor=<rgba 0.6406 1 0 1>}
  TextureCoordinate()
 ]}
