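# Master material for the WiseBarn backpack cosmetic, adding a crystal ball effect to the standard character shading.
# The 2023 character PreFX parent feeds material attributes into MF_WiseBarnCrystalBall, which also takes the TextureMask2 FX mask.
# The result passes through the matching PostFX parent alongside other PreFX outputs.
# It supports skeletal meshes, morph targets and mesh deformers, flags pixel animation, and selects its shading model by expression.
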Material /BRCosmetics/Accessories/FORT_Backpacks/Backpack_F_MED_WiseBarn/Materials/M_WiseBarn_Backpack
graph {
 properties {
  bHasPixelAnimation=true
  bUseMaterialAttributes=true
  bUsedWithMeshDeformer=true
  bUsedWithMorphTargets=true
  bUsedWithSkeletalMesh=true
  pixelDepthOffsetMode=PDOM_Legacy
  shadingModel=MSM_FromMaterialExpression
  usedShadingModels=Default Lit | Subsurface | Clear Coat | Subsurface Profile | Cloth
 }
 nodes [
  n1 = MaterialFunctionCall() {materialFunction=&MF_CharacterParent_2023_PreFX}
  n2 = TextureSampleParameter2D() {
   parameterName=TextureMask2
   samplerType=SAMPLERTYPE_Masks
   texture=/BRCosmetics/Accessories/FORT_Backpacks/Backpack_F_MED_WiseBarn/Textures/T_WiseBarn_Backpack_FX_TechArt
  }
  MaterialFunctionCall(
   Material=MaterialFunctionCall!(In_MaterialAttribute=@n1.Result, In_TechArtMask2=@n2.RGB) {materialFunction=&/MF_WiseBarnCrystalBall}.Result
   M=@n1.M
   Pre Skinned Local Position=@n1.M
   Distance Blend=@n1.M
   Fresnel=@n1.M
  ) {materialFunction=&MF_CharacterParent_2023_PostFX}
 ]
}
